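# Animation blueprint for the FishLine backpack, spinning its fish when an enemy player is eliminated, or on a timed loop for frontend display.
# The elimination event briefly raises a target rate; an interpolated PlayRate drives both the elimination sequence and accumulated pack_center rotation, then eases to zero.
# Additive pack bone offsets read variables computed from owner spine bend, but that calculation chain has no incoming execution link.
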AnimBlueprint /BRCosmetics/Accessories/FORT_Backpacks/Backpack_F_MED_FishLine/Meshes/F_MED_FishLine_Pack_AnimBP
target_skeleton=&/F_MED_FishLine_Pack_Skeleton
graphs [
 AnimGraph {nodes [
   Input Pose()
   n1 = Get PlayRate() -> (PlayRate:double)
   n2 = "Transform (Modify) Bone - Bone: pack"(
    ComponentPose=Local To Component(LocalPose=Sequence Player 'FishLine_Elim_Pack'!(PlayRate:float=@n1) -> (Pose) {sequence=/BRCosmetics/Accessories/FORT_Backpacks/Backpack_F_MED_FishLine/Anim/FishLine_Elim_Pack})
    Translation=Get backpack_translation()
    Rotation=Get backpack_rotate()
   ) {rotationMode=BMM_Additive, translationMode=BMM_Additive, alphaBoolBlend {blendOption=Linear}, boneToModify {boneName=pack}}
   n3 = "Transform (Modify) Bone - Bone: pack"(
    ComponentPose=@n2
    Translation=Get Backpack_translation_X() -> (Backpack_translation_X:Vector)
    Rotation=Get Backpack_rotate_X() -> (Backpack_rotate_X:Rotator)
   ) {rotationMode=BMM_Additive, translationMode=BMM_Additive, alphaBoolBlend {blendOption=Linear}, boneToModify {boneName=pack}}
   n4 = "Transform (Modify) Bone - Bone: pack"(ComponentPose=@n3, Translation="0,0,0", Rotation=Get Backpack_rotate_-X() -> (Backpack_rotate_-X:Rotator)) {rotationMode=BMM_Additive, translationMode=BMM_Additive, alphaBoolBlend {blendOption=Linear}, boneToModify {boneName=pack}}
   n5 = Get FishRotation() -> (FishRotation:Rotator)
   Output Pose(Result=Component To Local(ComponentPose="Transform (Modify) Bone - Bone: pack_center"(ComponentPose=@n4, Alpha=1, Translation="0,0,0", Rotation=@n5, Scale="1,1,1") {rotationMode=BMM_Additive, alphaBoolBlend {blendOption=Linear}, boneToModify {boneName=pack_center}})) {layerGroup=None}
  ]}
 EventGraph {nodes [
   n1 = Event BlueprintProcessAnimInputEvent()
   > n2 = Branch(Condition=IsNamed(self=@n1.AnimInputEvent, InEventName=aie.EnemyPlayerEliminated))
   n3 = Set PlayRate(execute=[@n5.then, @n2.then], PlayRate:double=0) -> (Output_Get:double)
   Set TargetPlayRate(execute=[@n3.then, @n4.then], TargetPlayRate:double=1.2) -> (Output_Get:double)
   > Delay(Duration=RandomFloatInRange(Min=2.75, Max=3.5))
   > Set TargetPlayRate(TargetPlayRate=0)
   > Branch(Condition=Get bShouldDisplayFrontendReactivity() -> (bShouldDisplayFrontendReactivity:bool))
   > n4 = Delay(Duration=12)
   Event BlueprintInitializeAnimation()
   > Branch(Condition=Get bShouldDisplayFrontendReactivity())
   > Do Once()
   > n5 = Delay(Duration=4)
   n6 = Event BlueprintUpdateAnimation()
   > Set Delta Time X(Delta Time X:double=@n6.DeltaTimeX) -> (Output_Get:double)
   > n7 = Sequence()
   > n8 = Branch(Condition="float < float"(A=Get PlayRate(), B=Get TargetPlayRate() -> (TargetPlayRate:double)))
   > Set PlayRate(PlayRate=FInterpTo_Constant!(
     Current:double=Get PlayRate()
     Target:double=Get TargetPlayRate()
     DeltaTime:double=Get Delta Time X() -> (Delta Time X:double)
     InterpSpeed:double=7.5
    ) -> (ReturnValue:double))
   Set PlayRate(
    execute=@n8.else
    PlayRate=FInterpTo_Constant(Current=Get PlayRate(), Target=Get TargetPlayRate(), DeltaTime=Get Delta Time X(), InterpSpeed=RandomFloatInRange(Min=0.75, Max=1.3))
   )
   n9 = float * float(A=float * float(A=Get PlayRate(), B=Get FishRotationMaxSpeed() -> (FishRotationMaxSpeed:double)), B=Get Delta Time X()) {is=title K2Node_PromotableOperator}
   n10 = Get RotationLastTickValue() -> (RotationLastTickValue:double)
   n11 = Set RotationValue(execute=@n7.then_1, RotationValue:double="% (Float)"!(A:double=float + float(A=@n9, B=@n10), B:double=360) -> (ReturnValue:double)) -> (Output_Get:double)
   > Set FishRotation(FishRotation:Rotator=MakeRotator(Pitch=@n11)) -> (Output_Get:Rotator)
   > Set RotationLastTickValue(RotationLastTickValue:double=Get RotationValue() -> (RotationValue:double)) -> (Output_Get:double)
   n12 = GetOwnerSkeletalMeshComponent()
   n13 = BreakRotator(InRot=BreakTransform(InTransform=GetDeltaTransformFromRefPose(self=@n12, BoneName=spine_05, BaseName=spine_01)).Rotation)
   n14 = MapRangeClamped(Value=@n13.Yaw, InRangeA=-30, InRangeB=-150, OutRangeB=230)
   n15 = MapRangeClamped(Value=@n13.Yaw, InRangeA=-8, InRangeB=-30, OutRangeB=10)
   Set backpack_rotate(backpack_rotate=MakeRotator(Roll=float + float(A=@n15, B=@n14)))
   n16 = MapRangeClamped(Value=@n13.Yaw, InRangeA=0, InRangeB=-20, OutRangeB=-3)
   n17 = MapRangeClamped(Value=@n13.Yaw, InRangeA=-20, InRangeB=-50, OutRangeB=-14)
   > Set backpack_translation(backpack_translation=MakeVector(Y=float + float(A=@n16, B=@n17), Z=MapRangeClamped(Value=@n13.Yaw, InRangeA=-30, InRangeB=-50, OutRangeB=-2)))
   > Set Backpack_rotate_X(Backpack_rotate_X:Rotator=MakeRotator(
     Pitch=MapRangeClamped(Value=@n13, InRangeA=10, InRangeB=45, OutRangeB=7)
     Yaw=MapRangeClamped(Value=@n13, InRangeA=10, InRangeB=45, OutRangeB=-20)
    )) -> (Output_Get:Rotator)
   > Set Backpack_translation_X(Backpack_translation_X:Vector=MakeVector(X=MapRangeClamped(Value=@n13, InRangeA=10, InRangeB=45, OutRangeB=-5), Y=0)) -> (Output_Get:Vector)
   > Set Backpack_rotate_-X(Backpack_rotate_-X:Rotator=MakeRotator(
     Roll=MapRangeClamped(Value=@n13, InRangeA=-10, InRangeB=-40, OutRangeB=18)
     Yaw=MapRangeClamped(Value=@n13, InRangeA=-10, InRangeB=-45, OutRangeB=6)
    )) -> (Output_Get:Rotator)
   TryGetPawnOwner()
   FInterpTo(
    Current=Get!() -> (BackpackRotate_Roll:float, BackpackRotate_Pitch:float, BackpackRotate_Yaw:float)
    Target=@n15
    DeltaTime=Reroute Node(InputPin:double) -> (OutputPin:double)
    InterpSpeed=50
   )
   FInterpTo(Current=Get().BackpackRotate_Pitch, Target=0, DeltaTime=0, InterpSpeed=50)
  ]}
 NewEventDispatcher {nodes [NewEventDispatcher() {is=title K2Node_FunctionEntry}]}
]
variables [
 backpack_rotate {type=Rotator}
 Backpack_rotate_-X {type=Rotator}
 Backpack_rotate_X {type=Rotator}
 backpack_translation {type=Vector}
 Backpack_translation_X {type=Vector}
 BackpackRotate {type=Rotator}
 Delta Time X {type=number}
 FishRotation {type=Rotator}
 FishRotationMaxSpeed {type=number, default=-90}
 NewEventDispatcher {default=()}
 PlayRate {type=number}
 RotationLastTickValue {type=number}
 RotationValue {type=number}
 TargetPlayRate {type=number}
]
